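; Editor utility actor that renders a preview of a static mesh into a render target through a SceneCapture2D.
; SetPreviewMesh assigns the mesh and calls FrameMesh, which places the capture on a raised diagonal, backed off by bounds size and a hardcoded FOV, aimed at the actor bounds.
; MakeCapture restricts the show-only list to the preview mesh and captures on demand.
; Other functions set the slot 0 material and get or set the render target.
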(event EventBeginPlay)

(event Collision|EventActorBeginOverlap
  (params (OtherActor :type Actor%20Object%20Reference)))

(event EventTick
  (params (DeltaSeconds :type Float%20%28single-precision%29)))

(fn FrameMesh
  (bind (_min _max) (Components|StaticMesh|GetLocalBounds (Variables|Default|GetPreviewMesh)))
  (Transformation|SetWorldLocationAndRotation (Variables|Default|GetSceneCapture) (+ (* (Math|Vector|Normalize (Math|Vector|MakeVector 1.0 1.0 0.5)) (* (/ (* (Math|Vector|VectorLength (Math|Vector|VectorGetAbs (- _max _min))) 0.5) (Math|Trig|Tan(Radians) (Math|Trig|DegreesToRadians (* 5e+01 0.5)))) 1.15)) (Collision|GetActorBounds false)) (Math|Rotator|FindLookatRotation (+ (* (Math|Vector|Normalize (Math|Vector|MakeVector 1.0 1.0 0.5)) (* (/ (* (Math|Vector|VectorLength (Math|Vector|VectorGetAbs (- _max _min))) 0.5) (Math|Trig|Tan(Radians) (Math|Trig|DegreesToRadians (* 5e+01 0.5)))) 1.15)) (Collision|GetActorBounds false)) (Collision|GetActorBounds false)) false true))

(fn GetRenderTarget
  (return (Variables|/Script/Engine.SceneCaptureComponent2D|GetTextureTarget (Variables|Default|GetSceneCapture))))

(fn MakeCapture
  (bind _scenecapture (Variables|Default|GetSceneCapture))
  (Rendering|SceneCapture|ClearShowOnlyComponents _scenecapture)
  (Rendering|SceneCapture|ShowOnlyComponent _scenecapture (Variables|Default|GetPreviewMesh))
  (Rendering|SceneCapture|CaptureScene _scenecapture :!class /Script/Engine.SceneCaptureComponent2D))

(fn SetPreviewMesh
  (params (Mesh :type Static%20Mesh%20Object%20Reference))
  (Components|StaticMesh|SetStaticMesh (Variables|Default|GetPreviewMesh) Mesh)
  (CallFunction|FrameMesh))

(fn SetPreviewMeshMaterial
  (params (Mat :type Material%20Interface%20Object%20Reference))
  (Rendering|Material|SetMaterial (Variables|Default|GetPreviewMesh) 0 Mat))

(fn SetRenderTarget
  (params (RT :type Texture%20Render%20Target%202D%20Object%20Reference))
  (Variables|/Script/Engine.SceneCaptureComponent2D|SetTextureTarget RT (Variables|Default|GetSceneCapture)))

(fn ConstructionScript)
